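# Niagara impact effect for the ForestJeanWink mic stand pickaxe, combining surface debris, a flash and glitter stars.
# Debris meshes take velocity and gravity from the User.PhysicalSurface response.
# Flash and glitter emitters use spawn groups to layer meshes, sprites, light, flares and twinkles; the original glitter emitter is disabled beside its active copy.
# User.IsVariant switches red and pink-purple palettes, while time of day modulates flash brightness.
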NiagaraSystem /SparksCosmetics/Instrument/MicStand/ForestJeanWink/Effects/NS_Pickaxe_MicStand_ForestJeanWink_Impact
dependencies {
 usedDataInterfaces [
  &NiagaraDataInterfaceCurve
  &NiagaraDataInterfaceArrayColor
  &NiagaraDataInterfaceVector2DCurve
  &NiagaraDataInterfaceColorCurve
 ]
 usedDynamicInputs []
 usedModules [
  &self:SystemSpawnScript.SystemScriptSource.SystemScriptGraph.NiagaraNodeAssignment_0.SetVariables_#1
  &SystemState
 ]
}
emitters {
 FN_PhysMatDebris {
  emitterUpdateScript [
   &EmitterState()
   &SpawnBurst_Instantaneous(
    Spawn Count=&RandomRangeInt(Minimum=5, Maximum=10)
    Use Spawn Probability=false
    UseLoopCountLimit=false
   )
  ]
  particleSpawnScript [
   /SetVariables(Particles.BiasedRandom=&FloatFromCurve(
     FloatCurve=DataInterface() {propertyValues {Curve {keys [
         <at 0.2 1 RCIM_Cubic>
         {arriveTangent=-3.3394, interpMode=RCIM_Cubic, leaveTangent=-3.3394, tangentMode=RCTM_User, time=0.5, value=0.5}
         <at 0.8 0 RCIM_Cubic>
        ]}}}
     CurveIndex=&RandomRangeFloat(Minimum=0, Maximum=1)
    )) {isSetParameters=true}
   &InitializeParticle(
    Color Mode=::NewEnumerator0
    Lifetime=&RandomRangeFloat(Minimum=0.5, Maximum=1.25)
    Mass=&Lerp_Float(A=0.6, B=1.2, Alpha=$Particles.BiasedRandom)
    Mass Mode=::NewEnumerator1
    Mesh Scale=&Multiply_VectorByFloat(
     Vector=&RandomRangeVector(Minimum=<0.9 0.4 0.5>, Maximum=<1 1 1>)
     Float=&Lerp_Float(A=0.7, B=1, Alpha=$Particles.BiasedRandom)
    )
    Mesh Scale Mode=::NewEnumerator4
    UsePositionOffset=false
    Write Material Random=true
   )
   &FN_PhysicalSurface(
    Dirt Mesh Count=2
    Glass Mesh Count=3
    Ice Mesh Count=2
    Metal Gravity=&RandomRangeFloat(Minimum=-350, Maximum=-1800)
    Metal Mesh Count=2
    Metal Velocity=&RandomRangeFloat(Minimum=80, Maximum=375)
    Metal Velocity Cone Angle=130
    Physical Surface=$User.PhysicalSurface
    Stone Gravity=&RandomRangeFloat(Minimum=-700, Maximum=-1250)
    Stone Mesh Count=2
    Stone Velocity=&RandomRangeFloat(Minimum=80, Maximum=550)
    Stone Velocity Cone Angle=120
    Wood Gravity=&RandomRangeFloat(Minimum=-450, Maximum=-900)
    Wood Mesh Count=2
    Wood Scale=1.5
    Wood Velocity=&RandomRangeFloat(Minimum=80, Maximum=600)
    Wood Velocity Cone Angle=130
   )
   &AddVelocityInCone(
    Cone Angle=$Output.FN_PhysicalSurface.SurfaceResponseConeAngle
    Module.Use Velocity Falloff On Cone Axis=false
    Velocity Strength=$Output.FN_PhysicalSurface.SurfaceResponseVelocity
   )
   &InitialMeshOrientation(Use Rotation Vector=false)
   &AddRotationalVelocity(Rotation Rate=&RandomVector(Vector Scale=&RandomRangeFloat(Minimum=1, Maximum=2)))
   &ApplyInitialForces()
   /SetVariables(Particles.GravityOverrideValue=$Output.FN_PhysicalSurface.GravityOverride) {isSetParameters=true}
  ]
  particleUpdateScript [
   &ParticleState()
   &GravityForce(Gravity=$Particles.GravityOverrideValue)
   &Collision(
    Advanced Aging Rate=1
    Age Colliding Particles=false
    CPU Collision Trace Channel=ECC_WorldStatic
    Collision Enabled=&Boolean_AndOperation(
     Bool1=&SetBoolByFloatComparison(A=$Particles.Initial.Lifetime, B=1)
     Bool2=&SetBoolByFloatComparison(A=$Particles.Age, B=0.12)
    )
    Control Roll On Collision=true
    Friction=1
    Friction During a Bounce=0.25
    Max CPU Trace Length=100
    Module.Control Rest State Via Surface Normal=false
    Module.EnableMaxCollisionCount=false
    Module.Randomize Collision Normal Vector=true
    Randomize Collision Normal=0.25
    ScaleRateOfRotation=false
   ) {enabled=false}
   &SolveForcesAndVelocity(Clamp Velocity=false, Limit Acceleration=false)
   &ScaleMeshSize(Scale Factor=&VectorFromFloat(Value=&FloatFromCurve(FloatCurve=DataInterface() {propertyValues {Curve {keys [<at 0.6 1 RCIM_Cubic>, {arriveTangent=-6.308, interpMode=RCIM_Cubic, leaveTangent=-6.308, tangentMode=RCTM_User, time=1, value=0}]}}})))
   &DynamicMaterialParameters(Module.Param0WriteEnabled=true, Module.Param1WriteEnabled=true, Module.Param2WriteEnabled=false, Module.Param3WriteEnabled=false)
  ]
  rendererClasses [&NiagaraSpriteRendererProperties, &NiagaraMeshRendererProperties]
  renderers [
   {rendererClass=&NiagaraSpriteRendererProperties}
   {rendererClass=&NiagaraMeshRendererProperties, rendererIndex=1}
   {rendererClass=&NiagaraMeshRendererProperties, rendererIndex=2}
   {rendererClass=&NiagaraMeshRendererProperties, rendererIndex=3}
   {rendererClass=&NiagaraMeshRendererProperties, rendererIndex=4}
   {rendererClass=&NiagaraMeshRendererProperties, rendererIndex=5}
   {rendererClass=&NiagaraMeshRendererProperties, rendererIndex=6}
   {rendererClass=&NiagaraMeshRendererProperties, rendererIndex=7}
   {rendererClass=&NiagaraMeshRendererProperties, rendererIndex=8}
  ]
 }
 Flash {
  emitterUpdateScript [
   &EmitterState()
   &SpawnBurst_Instantaneous(Spawn Count=1, Use Spawn Probability=false)
   &SpawnBurst_Instantaneous(Spawn Count=&RandomRangeInt(Minimum=7, Maximum=9), Spawn Group=1, Use Spawn Probability=false) {module=SpawnBurst_Instantaneous001}
   &SpawnBurst_Instantaneous(Spawn Count=5, Spawn Group=2, Use Spawn Probability=false, UseLoopCountLimit=false) {module=SpawnBurst_Instantaneous002}
   &SpawnBurst_Instantaneous(Spawn Group=3, Use Spawn Probability=false, UseLoopCountLimit=false) {module=SpawnBurst_Instantaneous003}
   &SpawnBurst_Instantaneous(Spawn Group=4, Use Spawn Probability=false, UseLoopCountLimit=false) {module=SpawnBurst_Instantaneous004}
   &SpawnBurst_Instantaneous(Spawn Group=5, Use Spawn Probability=false, UseLoopCountLimit=false) {module=SpawnBurst_Instantaneous005}
  ]
  particleSpawnScript [
   /SetVariables(Particles.VisibilityTag=&SpawnGroup()) {isSetParameters=true}
   &InitializeParticle(
    Color=&MakeCustomLinearColorFromBool(
     Bool=&SetBoolByFloatComparison(A=$User.IsVariant, B=0.01)
     True Color=&MergeHelper_ColorSwitch(
      0=<rgba 2 0 0.9124 1>
      1=<rgba 1 0.0306 0.1008 1>
      2=<rgba 0.6458 0.0136 0.0692 1>
      3=<rgba 1.5 0.1767 0.2312 1>
      4=<rgba 1 1 1 1>
      5=<rgba 0.0443 0 0.0004 0.75>
      Index=&SpawnGroup()
      Num=::NewEnumerator4
     )
     False Color=&MergeHelper_ColorSwitch(
      0=<rgba 2 0.4778 1.3047 1>
      1=<rgba 1 0.25 0.8305 1>
      2=<rgba 0.6458 0.1807 0.4598 1>
      3=<rgba 1.5 0.5742 1.0297 1>
      4=<rgba 1 1 1 1>
      5=<rgba 0.0625 0 0.0366 0.75>
      Index=&SpawnGroup()
      Num=::NewEnumerator4
     )
    )
    Lifetime=&MergeHelper_FloatSwitch(
     0=0.2
     1=&RandomRangeFloat(Minimum=0.1, Maximum=0.13)
     2=&RandomRangeFloat(Minimum=0.1, Maximum=0.14)
     3=0.18
     4=0.15
     5=0.23
     Index=&SpawnGroup()
     Num=::NewEnumerator4
    )
    Mesh Scale=&Multiply_VectorByFloat(Vector=&RandomRangeVector(Minimum=<1 2 2>, Maximum=<0.8 1.75 1.75>), Float=$System.GlobalScale)
    Mesh Scale Mode=::NewEnumerator4
    Sprite Size=&Multiply_Vector2D_ByFloat(
     Float=$System.GlobalScale
     Vector2D=&MergeHelper_Vector2Switch(
      0=<1 1>
      1=&RandomRangeVector2D(Minimum=<9 45>, Maximum=<15 60>)
      2=&RandomRangeVector2D(Minimum=<30 45>, Maximum=<45 75>)
      3=&Vector2DFromFloat(Value=200)
      4=<1 1>
      5=&Vector2DFromFloat(Value=125)
      Index=$Emitter.SpawnGroup
      Num=::NewEnumerator4
     )
    )
    Sprite Size Mode=::NewEnumerator4
    UsePositionOffset=false
    Write Material Random=true
   )
   &ShapeLocation(
    MaskSpawnGroup=false
    Override Seed=false
    Rotation Angle:NiagaraFloat=&RandomRangeFloat(Minimum=0, Maximum=1)
    Rotation Mode=::NewEnumerator1
    Sphere Radius=&MergeHelper_FloatSwitch(
     0=0
     1=&Multiply_Float(A=6, B=$System.GlobalScale)
     2=&Multiply_Float(A=6, B=$System.GlobalScale)
     3=0
     4=0
     5=0
     Index=&SpawnGroup()
     Num=::NewEnumerator4
    )
    UseOwnerScale=true
   )
   &DynamicMaterialParameters(
    Index 0 Param 2=45
    Index 0 Param 3=1
    Index 0 Param 4=1
    Module.Param0WriteEnabled=false
    Module.Param1WriteEnabled=true
    Module.Param2WriteEnabled=true
    Module.Param3WriteEnabled=true
   ) {module=DynamicMaterialParameters001}
   &InitialMeshOrientation(
    Rotation=&RandomRangeVector(Minimum=<-1 -1 -1>, Maximum=<1 1 1>)
    Use Orientation Vector=true
    Use Rotation Vector=true
   )
   &CameraOffset(Camera Offset Amount=&MergeHelper_FloatSwitch(0=20, 1=10, 2=10, 3=25, 4=0, 5=25, Index=&SpawnGroup(), Num=::NewEnumerator4))
   &AddVelocityFromPoint(
    Module.Use Falloff Distance=false
    Velocity Strength=&MergeHelper_FloatSwitch(0=0, 1=100, 2=125, 3=0, 4=0, Index=&SpawnGroup(), Num=::NewEnumerator3)
   )
  ]
  particleUpdateScript [
   &ParticleState()
   &DynamicMaterialParameters(
    Index 0 Param 1=&FloatFromCurve(FloatCurve=DataInterface() {propertyValues {Curve {keys [<at 0.2 0 RCIM_Cubic>, {arriveTangent=3.6451, interpMode=RCIM_Cubic, leaveTangent=3.6451, tangentMode=RCTM_User, time=1, value=1}]}}})
    Module.Index0Float0Write=true
    Module.Index0Float1Write=true
    Module.Index0Float2Write=true
    Module.Index0Float3Write=true
   )
   &ScaleMeshSize(Scale Factor=&VectorFromFloat(Value=&FloatFromCurve(
      FloatCurve=DataInterface() {propertyValues {Curve {
         defaultValue=0.5
         keys [
          {arriveTangent=2.9597, interpMode=RCIM_Cubic, leaveTangent=2.9597, tangentMode=RCTM_User, time=0, value=0}
          {arriveTangent=0.721, interpMode=RCIM_Cubic, leaveTangent=0.721, tangentMode=RCTM_User, time=1, value=1}
         ]
        }}}
      Scale Curve=0.7
     )))
   &ScaleColor(
    Scale Alpha=&MergeHelper_FloatSwitch(
     0=&FloatFromCurve(FloatCurve=DataInterface() {propertyValues {Curve {keys [<at 0 1>,<at 1 0>]}}})
     1=&FloatFromCurve(FloatCurve=DataInterface() {propertyValues {Curve {
         defaultValue=1
         keys [{interpMode=RCIM_Cubic, tangentMode=RCTM_User, time=0, value=1}, {interpMode=RCIM_Cubic, tangentMode=RCTM_User, time=1, value=0}]
        }}})
     2=&FloatFromCurve(FloatCurve=DataInterface() {propertyValues {Curve {
         defaultValue=1
         keys [
          {arriveTangent=-2.0249, interpMode=RCIM_Cubic, leaveTangent=-2.0249, tangentMode=RCTM_User, time=0, value=1}
          {interpMode=RCIM_Cubic, tangentMode=RCTM_User, time=1, value=0}
         ]
        }}})
     3=&FloatFromCurve(
      FloatCurve=DataInterface() {propertyValues {Curve {
         defaultValue=1
         keys [
          {interpMode=RCIM_Cubic, tangentMode=RCTM_User, time=0, value=1}
          {arriveTangent=-1.2989, interpMode=RCIM_Cubic, leaveTangent=-1.2989, tangentMode=RCTM_User, time=1, value=0}
         ]
        }}}
      Scale Curve=&FN_FrontEndFloatModifier(
       Gameplay=&FloatFromCurve(
        FloatCurve=DataInterface() {propertyValues {Curve {keys [
            <at 0 0.65 RCIM_Cubic>
            {arriveTangent=0.875, interpMode=RCIM_Cubic, leaveTangent=0.875, time=0.1, value=0.7}
            {arriveTangentWeight=0.1329, interpMode=RCIM_Cubic, leaveTangentWeight=0.0759, tangentMode=RCTM_User, time=0.4, value=1}
            {arriveTangentWeight=0.0801, interpMode=RCIM_Cubic, leaveTangentWeight=0.1002, tangentMode=RCTM_User, time=0.6, value=1}
            {arriveTangent=-0.8778, interpMode=RCIM_Cubic, leaveTangent=-0.8778, time=0.9, value=0.7}
            <at 0.9987 0.65 RCIM_Cubic>
           ]}}}
        CurveIndex=Linked() {linkedVariable {name=NPC.FortniteNPC.TimeOfDay_0-1, type=NiagaraFloat}}
       )
       Front End=0.5
      )
     )
     4=1
     5=&FloatFromCurve(
      FloatCurve=DataInterface() {propertyValues {Curve {
         defaultValue=1
         keys [
          {interpMode=RCIM_Cubic, tangentMode=RCTM_User, time=0, value=1}
          {arriveTangent=-2.0745, interpMode=RCIM_Cubic, leaveTangent=-2.0745, tangentMode=RCTM_User, time=1, value=0}
         ]
        }}}
      Scale Curve=&FloatFromCurve(
       FloatCurve=DataInterface() {propertyValues {Curve {
          defaultValue=1
          keys [
           {interpMode=RCIM_Cubic, tangentMode=RCTM_User, time=0, value=0.45}
           {interpMode=RCIM_Cubic, tangentMode=RCTM_User, time=0.5, value=1}
           {tangentMode=RCTM_User, time=1, value=0.45}
          ]
         }}}
       CurveIndex=Linked() {linkedVariable {name=NPC.FortniteNPC.TimeOfDay_0-1, type=NiagaraFloat}}
       Scale Curve=1.2
      )
     )
     Index=&SpawnGroup()
     Num=::NewEnumerator4
    )
    Scale RGB=&MakeCustomVectorFromBool(
     Bool=&SetBoolByFloatComparison(A=$User.IsVariant, B=0)
     True Vector=&MergeHelper_VectorSwitch(
      0=&Multiply_VectorByFloat(
       Vector=&MakeVectorFromLinearColorRGB(Color=&ColorFromCurve(ColorCurve=DataInterface() {propertyValues {
           AlphaCurve {keys []}
           BlueCurve {keys [<at 0.032 0.6302 RCIM_Cubic>,<at 0.3111 0.4835>,<at 0.833 0.5964 RCIM_Cubic>]}
           GreenCurve {keys [<at 0.032 0 RCIM_Cubic>,<at 0.3111 0>,<at 0.833 0 RCIM_Cubic>]}
           RedCurve {keys [<at 0.032 6.07 RCIM_Cubic>,<at 0.3111 2>,<at 0.833 1 RCIM_Cubic>]}
          }}))
       Float=&FloatFromCurve(
        FloatCurve=DataInterface() {propertyValues {Curve {keys [<at 0 4>,<at 0.2 1.5>,<at 1 1>]}}}
        Scale Curve=&FloatFromCurve(
         FloatCurve=DataInterface() {propertyValues {Curve {keys [
             <at 0.1 0.8 RCIM_Cubic>
             {arriveTangentWeight=0.1329, interpMode=RCIM_Cubic, leaveTangentWeight=0.0759, tangentMode=RCTM_User, time=0.35, value=1}
             {arriveTangentWeight=0.0801, interpMode=RCIM_Cubic, leaveTangentWeight=0.1002, tangentMode=RCTM_User, time=0.65, value=1}
             <at 0.9 0.8 RCIM_Cubic>
            ]}}}
         CurveIndex=Linked() {linkedVariable {name=NPC.FortniteNPC.TimeOfDay_0-1, type=NiagaraFloat}}
        )
       )
      )
      1=&Multiply_VectorByFloat(
       Vector=&MakeVectorFromLinearColorRGB(Color=&ColorFromCurve(ColorCurve=DataInterface() {propertyValues {
           AlphaCurve {keys []}
           BlueCurve {keys [<at 0 2.953 RCIM_Cubic>,<at 0.2723 1.2773>,<at 0.4312 0.3827>,<at 1 0.2554 RCIM_Cubic>]}
           GreenCurve {keys [<at 0 2.1129 RCIM_Cubic>,<at 0.2723 0.9115>,<at 0.4312 0.2448>,<at 1 0.1172 RCIM_Cubic>]}
           RedCurve {keys [<at 0 7 RCIM_Cubic>,<at 0.2723 3>,<at 0.4312 1>,<at 1 1 RCIM_Cubic>]}
          }}))
       Float=&FloatFromCurve(
        FloatCurve=DataInterface() {propertyValues {Curve {keys [
            <at 0.1 0.8 RCIM_Cubic>
            {arriveTangentWeight=0.1329, interpMode=RCIM_Cubic, leaveTangentWeight=0.0759, tangentMode=RCTM_User, time=0.45, value=1}
            {arriveTangentWeight=0.0801, interpMode=RCIM_Cubic, leaveTangentWeight=0.1002, tangentMode=RCTM_User, time=0.55, value=1}
            <at 0.9 0.8 RCIM_Cubic>
           ]}}}
        CurveIndex=Linked() {linkedVariable {name=NPC.FortniteNPC.TimeOfDay_0-1, type=NiagaraFloat}}
       )
      )
      2=&VectorFromFloat(Value=0.4)
      3=&Multiply_VectorByFloat(
       Vector=&MakeVectorFromLinearColorRGB(Color=&ColorFromCurve(ColorCurve=DataInterface() {propertyValues {
           AlphaCurve {keys [<at 0 1 RCIM_Cubic>]}
           BlueCurve {keys [<at 0.1188 0 RCIM_Cubic>,<at 0.5059 0.003>,<at 1 0.0522 RCIM_Cubic>]}
           GreenCurve {keys [<at 0.1188 0 RCIM_Cubic>,<at 0.5059 0.003>,<at 1 0 RCIM_Cubic>]}
           RedCurve {keys [<at 0.1188 3 RCIM_Cubic>,<at 0.5059 1.5>,<at 1 1 RCIM_Cubic>]}
          }}))
       Float=&FloatFromCurve(
        FloatCurve=DataInterface() {propertyValues {Curve {keys [
            <at 0.1 0.75 RCIM_Cubic>
            {arriveTangentWeight=0.1329, interpMode=RCIM_Cubic, leaveTangentWeight=0.0759, tangentMode=RCTM_User, time=0.45, value=1}
            {arriveTangentWeight=0.0801, interpMode=RCIM_Cubic, leaveTangentWeight=0.1002, tangentMode=RCTM_User, time=0.55, value=1}
            <at 0.9 0.75 RCIM_Cubic>
           ]}}}
        CurveIndex=Linked() {linkedVariable {name=NPC.FortniteNPC.TimeOfDay_0-1, type=NiagaraFloat}}
       )
      )
      4=<1 1 1>
      5=<1 1 1>
      Index=&SpawnGroup()
      Num=::NewEnumerator4
     )
     False Vector=&MergeHelper_VectorSwitch(
      0=&Multiply_VectorByFloat(
       Vector=&MakeVectorFromLinearColorRGB(Color=&ColorFromCurve(ColorCurve=DataInterface() {propertyValues {
           AlphaCurve {keys []}
           BlueCurve {keys [<at 0.0238 3.9497 RCIM_Cubic>,<at 0.1913 1.0239>,<at 0.833 0.9043 RCIM_Cubic>]}
           GreenCurve {keys [<at 0.0238 0.934 RCIM_Cubic>,<at 0.1913 0.5559>,<at 0.833 0.3124 RCIM_Cubic>]}
           RedCurve {keys [<at 0.0238 5 RCIM_Cubic>,<at 0.1913 1.5>,<at 0.833 1 RCIM_Cubic>]}
          }}))
       Float=&FloatFromCurve(
        FloatCurve=DataInterface() {propertyValues {Curve {keys [<at 0 4>,<at 0.2 1.5>,<at 1 1>]}}}
        Scale Curve=&FloatFromCurve(
         FloatCurve=DataInterface() {propertyValues {Curve {keys [
             <at 0.1 0.8 RCIM_Cubic>
             {arriveTangentWeight=0.1329, interpMode=RCIM_Cubic, leaveTangentWeight=0.0759, tangentMode=RCTM_User, time=0.35, value=1}
             {arriveTangentWeight=0.0801, interpMode=RCIM_Cubic, leaveTangentWeight=0.1002, tangentMode=RCTM_User, time=0.65, value=1}
             <at 0.9 0.8 RCIM_Cubic>
            ]}}}
         CurveIndex=Linked() {linkedVariable {name=NPC.FortniteNPC.TimeOfDay_0-1, type=NiagaraFloat}}
        )
       )
      )
      1=&Multiply_VectorByFloat(
       Vector=&MakeVectorFromLinearColorRGB(Color=&ColorFromCurve(ColorCurve=DataInterface() {propertyValues {
           AlphaCurve {keys []}
           BlueCurve {keys [<at 0 5 RCIM_Cubic>,<at 0.4312 0.8755>,<at 1 0.5981 RCIM_Cubic>]}
           GreenCurve {keys [<at 0 2.8564 RCIM_Cubic>,<at 0.4312 0.3906>,<at 1 0.2031 RCIM_Cubic>]}
           RedCurve {keys [<at 0 4.0357 RCIM_Cubic>,<at 0.4312 1>,<at 1 1 RCIM_Cubic>]}
          }}))
       Float=&FloatFromCurve(
        FloatCurve=DataInterface() {propertyValues {Curve {keys [
            <at 0.1 0.8 RCIM_Cubic>
            {arriveTangentWeight=0.1329, interpMode=RCIM_Cubic, leaveTangentWeight=0.0759, tangentMode=RCTM_User, time=0.45, value=1}
            {arriveTangentWeight=0.0801, interpMode=RCIM_Cubic, leaveTangentWeight=0.1002, tangentMode=RCTM_User, time=0.55, value=1}
            <at 0.9 0.8 RCIM_Cubic>
           ]}}}
        CurveIndex=Linked() {linkedVariable {name=NPC.FortniteNPC.TimeOfDay_0-1, type=NiagaraFloat}}
       )
      )
      2=&VectorFromFloat(Value=0.4)
      3=&Multiply_VectorByFloat(
       Vector=&MakeVectorFromLinearColorRGB(Color=&ColorFromCurve(ColorCurve=DataInterface() {propertyValues {
           AlphaCurve {keys [<at 0 1 RCIM_Cubic>]}
           BlueCurve {keys [<at 0.1188 1.6012 RCIM_Cubic>,<at 0.5059 0.7106>,<at 1 0.6323 RCIM_Cubic>]}
           GreenCurve {keys [<at 0.1188 0.1042 RCIM_Cubic>,<at 0.5059 0.0469>,<at 1 0 RCIM_Cubic>]}
           RedCurve {keys [<at 0.1188 4 RCIM_Cubic>,<at 0.5059 1.5>,<at 1 1 RCIM_Cubic>]}
          }}))
       Float=&FloatFromCurve(
        FloatCurve=DataInterface() {propertyValues {Curve {keys [
            <at 0.1 0.75 RCIM_Cubic>
            {arriveTangentWeight=0.1329, interpMode=RCIM_Cubic, leaveTangentWeight=0.0759, tangentMode=RCTM_User, time=0.45, value=0.95}
            {arriveTangentWeight=0.0801, interpMode=RCIM_Cubic, leaveTangentWeight=0.1002, tangentMode=RCTM_User, time=0.55, value=0.95}
            <at 0.9 0.75 RCIM_Cubic>
           ]}}}
        CurveIndex=Linked() {linkedVariable {name=NPC.FortniteNPC.TimeOfDay_0-1, type=NiagaraFloat}}
       )
      )
      4=<1 1 1>
      5=<1 1 1>
      Index=&SpawnGroup()
      Num=::NewEnumerator4
     )
    )
    ScaleA=true
    ScaleRGB=true
   )
   &ScaleSpriteSize(
    Scale Factor=&MergeHelper_Vector2Switch(
     0=<0 0>
     1=&MakeVector2D(
      X=&FloatFromCurve(FloatCurve=DataInterface() {propertyValues {Curve {
          defaultValue=1
          keys [
           {interpMode=RCIM_Cubic, tangentMode=RCTM_User, time=0, value=0}
           {arriveTangentWeight=0.2236, interpMode=RCIM_Cubic, leaveTangentWeight=0.1491, tangentMode=RCTM_User, time=0.7, value=1}
           {arriveTangent=-2.2754, interpMode=RCIM_Cubic, leaveTangent=-2.2754, tangentMode=RCTM_User, time=1, value=0.5}
          ]
         }}})
      Y=&FloatFromCurve(FloatCurve=DataInterface() {propertyValues {Curve {
          defaultValue=1
          keys [
           {arriveTangent=3.143, interpMode=RCIM_Cubic, leaveTangent=3.143, tangentMode=RCTM_User, time=0, value=0}
           {arriveTangent=0.8537, interpMode=RCIM_Cubic, leaveTangent=0.8537, tangentMode=RCTM_User, time=1, value=1}
          ]
         }}})
     )
     2=&Vector2DFromFloat(Value=&FloatFromCurve(FloatCurve=DataInterface() {propertyValues {Curve {
          defaultValue=1
          keys [
           {arriveTangent=0.9682, interpMode=RCIM_Cubic, leaveTangent=0.9682, tangentMode=RCTM_User, time=0, value=0}
           {interpMode=RCIM_Cubic, tangentMode=RCTM_User, time=1, value=1}
          ]
         }}}))
     3=&Multiply_Vector2D_ByFloat(
      Float=&FloatFromCurve(FloatCurve=DataInterface() {propertyValues {Curve {
          defaultValue=0
          keys [
           {interpMode=RCIM_Cubic, tangentMode=RCTM_User, time=0, value=1}
           {arriveTangent=-1.4808, interpMode=RCIM_Cubic, leaveTangent=-1.4808, tangentMode=RCTM_User, time=1, value=0}
          ]
         }}})
      Vector2D=<1 1>
     )
     4=<0 0>
     5=&Vector2DFromFloat(Value=&FloatFromCurve(FloatCurve=DataInterface() {propertyValues {Curve {
          defaultValue=0
          keys [
           {arriveTangent=2.4894, interpMode=RCIM_Cubic, leaveTangent=2.4894, tangentMode=RCTM_User, time=0, value=0}
           {interpMode=RCIM_Cubic, tangentMode=RCTM_User, time=1, value=1}
          ]
         }}}))
     Index=&SpawnGroup()
     Num=::NewEnumerator4
    )
    Scale Sprite Size Mode=::NewEnumerator2
   )
   &SolveForcesAndVelocity(Clamp Velocity=false, Limit Acceleration=false)
   /SetVariables(Particles.Flare Dyn 1:Vector4f=&MakeVector4(
     X=&FloatFromCurve(FloatCurve=DataInterface() {propertyValues {Curve {keys [<at 0 1 RCIM_Cubic>,<at 0.8 0 RCIM_Cubic>]}}})
     Y=2
     Z=1
     W=1
    )) {isSetParameters=true}
   &FN_Light_Attributes(
    Color=&MultiplyLinearColor(
     Linear Color=$System.TOD Color
     Scale Factor=&MakeVector4_FromLinearColor(LinearColor=&MakeCustomLinearColorFromBool(
       Bool=&SetBoolByFloatComparison(A=$User.IsVariant, B=0)
       True Color=<rgba 1 0.0924 0.1574 1>
       False Color=<rgba 1 0.4169 0.7178 1>
      ))
    )
    Create Color Attribute=true
    Create Exponent Attribute=true
    Create Radius Attribute=true
    Exponent=3
    Radius=95
    Scale Color=&FloatFromCurve(
     FloatCurve=DataInterface() {propertyValues {Curve {
        defaultValue=1
        keys [{interpMode=RCIM_Cubic, tangentMode=RCTM_User, time=0, value=1}, {interpMode=RCIM_Cubic, tangentMode=RCTM_User, time=1, value=0}]
       }}}
     Scale Curve=12
    )
   )
  ]
  rendererClasses [
   &NiagaraMeshRendererProperties
   &NiagaraSpriteRendererProperties
   &NiagaraLightRendererProperties
  ]
  renderers [
   {rendererClass=&NiagaraMeshRendererProperties, rendererIndex=0}
   {rendererClass=&NiagaraSpriteRendererProperties, rendererIndex=1}
   {rendererClass=&NiagaraSpriteRendererProperties, rendererIndex=2}
   {rendererClass=&NiagaraSpriteRendererProperties, rendererIndex=3}
   {rendererClass=&NiagaraLightRendererProperties, rendererIndex=4}
   {rendererClass=&NiagaraSpriteRendererProperties, rendererIndex=5}
  ]
 }
 Glitter_Stars_Flares {
  bEnabled=false
  emitterUpdateScript [
   &EmitterState()
   &SpawnBurst_Instantaneous(
    Spawn Count=&RandomRangeInt(Minimum=8, Maximum=10)
    Use Spawn Probability=false
    UseLoopCountLimit=true
   )
   &SpawnBurst_Instantaneous(
    Spawn Count=&RandomRangeInt(Minimum=3, Maximum=5)
    Spawn Group=1
    Use Spawn Probability=false
    UseLoopCountLimit=true
   ) {module=SpawnBurst_Instantaneous001}
   &SpawnBurst_Instantaneous(Spawn Count=17, Spawn Group=2, Use Spawn Probability=false, UseLoopCountLimit=true) {module=SpawnBurst_Instantaneous002}
  ]
  particleSpawnScript [
   /SetVariables(Particles.VisibilityTag=&SpawnGroup()) {isSetParameters=true, module=SetVariables_#2}
   /SetVariables(Particles.Variance:NiagaraFloat=&MergeHelper_FloatSwitch(
     0=&FloatFromCurve(
      FloatCurve=DataInterface() {propertyValues {Curve {keys [<at 0 1>,<at 1 0>]}}}
      CurveIndex=&ReturnNormalizedExecIndex()
     )
     1=&FloatFromCurve(
      FloatCurve=DataInterface() {propertyValues {Curve {keys [<at 0 1>,<at 1 0>]}}}
      CurveIndex=&ReturnNormalizedExecIndex()
     )
     2=0
     3=?
     Index=&SpawnGroup()
     Num=::NewEnumerator1
    )) {isSetParameters=true, module=SetVariables_#3}
   &InitializeParticle(
    Color=&MergeHelper_ColorSwitch(
     0=&MakeCustomLinearColorFromBool(
      Bool=&SetBoolByFloatComparison(A=$User.IsVariant, B=0.01)
      True Color=&ColorFromCurve(ColorCurve=DataInterface() {propertyValues {
         AlphaCurve {keys [<at 0 1 RCIM_Cubic>]}
         BlueCurve {keys [
           <at 0.1342 0.2995>
           {arriveTangent=-0.1943, interpMode=RCIM_Cubic, leaveTangent=-0.1943, time=0.4451, value=0}
           <at 0.7943 0.1712 RCIM_Cubic>
          ]}
         GreenCurve {keys [
           <at 0.1342 0.2995>
           {arriveTangent=-0.4537, interpMode=RCIM_Cubic, leaveTangent=-0.4537, time=0.4451, value=0.0037}
           <at 0.7943 0 RCIM_Cubic>
          ]}
         RedCurve {keys [<at 0.1342 5>, {arriveTangent=-5.3023, interpMode=RCIM_Cubic, leaveTangent=-5.3023, time=0.4451, value=3}, <at 0.7943 1.5 RCIM_Cubic>]}
        }})
      False Color=&ColorFromCurve(ColorCurve=DataInterface() {propertyValues {
         AlphaCurve {keys [<at 0 1 RCIM_Cubic>]}
         BlueCurve {keys [<at 0.1342 3.3207>,<at 0.521 1.9924 RCIM_Cubic>,<at 0.7352 1.1029 RCIM_Cubic>]}
         GreenCurve {keys [<at 0.1342 0.5581>,<at 0.521 0.3349 RCIM_Cubic>,<at 0.7352 0 RCIM_Cubic>]}
         RedCurve {keys [<at 0.1342 5>,<at 0.521 3 RCIM_Cubic>,<at 0.7352 1.5 RCIM_Cubic>]}
        }})
     )
     1=&MakeCustomLinearColorFromBool(
      Bool=&SetBoolByFloatComparison(A=$User.IsVariant, B=0.01)
      True Color=<rgba 1.15 0.1096 0.092 1>
      False Color=<rgba 1 0.1274 0.402 1>
     )
     2=&MakeCustomLinearColorFromBool(
      Bool=&SetBoolByFloatComparison(A=$User.IsVariant, B=0)
      True Color=&SelectLinearColorFromArray(
       Array Sampling Mode=::NewEnumerator0
       Direct Array Index=?
       Color Selection Array=DataInterface() {propertyValues {ColorData [
          <rgba 1 0.1199 0.0714 1>
          <rgba 1 0.0571 0.0571 1>
          <rgba 0.1473 0.1405 0.149 1>
          <rgba 1 0.112 0.112 1>
          <rgba 0.2406 0.0526 0.0835 1>
          <rgba 1 0.1043 0.1043 1>
          <rgba 0.7698 0.0003 0.0404 1>
          <rgba 0.4281 0 0 1>
         ]}}
       Random Seed=0
      )
      False Color=&SelectLinearColorFromArray(
       Array Sampling Mode=::NewEnumerator0
       Direct Array Index=?
       Color Selection Array=DataInterface() {propertyValues {ColorData [
          <rgba 1 0.0714 0.5244 1>
          <rgba 0.5687 0.0571 1 1>
          <rgba 0.1473 0.1405 0.149 1>
          <rgba 0.8187 0.1885 1 1>
          <rgba 0.1226 0.0724 0.2406 1>
          <rgba 0.4133 0.1594 1 1>
          <rgba 0.7698 0.0003 0.6263 1>
          <rgba 0.4281 0 0.062 1>
         ]}}
       Random Seed=0
      )
     )
     3=?
     Index=&SpawnGroup()
     Num=::NewEnumerator1
    )
    Lifetime=&MergeHelper_FloatSwitch(
     0=&RandomRangeFloat(Minimum=0.45, Maximum=1)
     1=&FloatFromCurve(FloatCurve=DataInterface() {propertyValues {Curve {keys [<at 0 0.55>,<at 1 0.3>]}}}, CurveIndex=$Particles.Variance)
     2=&RandomRangeFloat(Minimum=0.5, Maximum=0.8)
     3=?
     Index=&SpawnGroup()
     Num=::NewEnumerator1
    )
    Sprite Rotation Angle=&MergeHelper_FloatSwitch(
     0=45
     1=&RandomRangeFloat(Minimum=0, Maximum=360)
     2=0
     3=?
     Index=&SpawnGroup()
     Num=::NewEnumerator1
    )
    Sprite Rotation Mode=::NewEnumerator0
    Sprite Size Mode=::NewEnumerator3
    Uniform Sprite Size=&MergeHelper_FloatSwitch(
     0=&RandomRangeFloat(Minimum=5, Maximum=9)
     1=&FloatFromCurve(
      FloatCurve=DataInterface() {propertyValues {Curve {keys [<at 0 5>,<at 1 9>]}}}
      CurveIndex=$Particles.Variance
      Scale Curve=&FN_FrontEndFloatModifier(Gameplay=1.9, Front End=1.5)
     )
     2=&Multiply_Float(A=&RandomRangeFloat(Minimum=1.5, Maximum=5.5), B=$System.GlobalScale)
     3=?
     Index=&SpawnGroup()
     Num=::NewEnumerator1
    )
    UsePositionOffset=false
    Write Material Random=true
   )
   &ShapeLocation(
    MaskSpawnGroup=false
    Rotation Angle:NiagaraFloat=&RandomRangeFloat(Minimum=0, Maximum=1)
    Rotation Axis=<0 1 1>
    Rotation Mode=::NewEnumerator1
    Sphere Distribution=::NewEnumerator3
    Sphere Radius=&Multiply_Float(
     A=&MergeHelper_FloatSwitch(0=15, 1=20, 2=7, 3=?, Index=&SpawnGroup(), Num=::NewEnumerator1)
     B=$System.GlobalScale
    )
    UseOwnerScale=true
   )
   &AddVelocity(
    Module.Use Falloff Distance=false
    Override Seed=false
    Velocity Mode=::NewEnumerator1
    Velocity Speed=&Multiply_Float(
     A=&MergeHelper_FloatSwitch(
      0=&RandomRangeFloat(Minimum=75, Maximum=200)
      1=&FloatFromCurve(FloatCurve=DataInterface() {propertyValues {Curve {keys [<at 0 175>,<at 1 100>]}}}, CurveIndex=$Particles.Variance)
      2=&RandomRangeFloat(Minimum=75, Maximum=150)
      3=?
      Index=&SpawnGroup()
      Num=::NewEnumerator1
     )
     B=$System.GlobalScale
    )
   )
   &CameraOffset(Camera Offset Amount=15)
  ]
  particleUpdateScript [
   &ParticleState()
   &ScaleSpriteSize(
    Scale Sprite Size Mode=::NewEnumerator0
    Uniform Scale Factor=&MergeHelper_FloatSwitch(
     0=&FloatFromCurve(
      FloatCurve=DataInterface() {propertyValues {Curve {
         defaultValue=1
         keys [
          {arriveTangent=9.0725, interpMode=RCIM_Cubic, leaveTangent=9.0725, tangentMode=RCTM_User, time=0, value=0}
          {arriveTangent=-2.5249, interpMode=RCIM_Cubic, leaveTangent=-2.5249, tangentMode=RCTM_User, time=0.1408, value=0.9847}
          {arriveTangent=-0.0405, interpMode=RCIM_Cubic, leaveTangent=-0.0405, tangentMode=RCTM_User, time=1, value=0}
         ]
        }}}
      Scale Curve=$System.GlobalScale
     )
     1=&FloatFromCurve(
      FloatCurve=DataInterface() {propertyValues {Curve {
         defaultValue=1
         keys [
          {arriveTangent=1.6536, interpMode=RCIM_Cubic, leaveTangent=1.6536, tangentMode=RCTM_User, time=0, value=0.2}
          {arriveTangent=0.2776, interpMode=RCIM_Cubic, leaveTangent=0.2776, tangentMode=RCTM_User, time=1, value=1}
         ]
        }}}
      Scale Curve=$System.GlobalScale
     )
     2=&FloatFromCurve(
      FloatCurve=DataInterface() {propertyValues {Curve {
         defaultValue=0
         keys [
          {interpMode=RCIM_Cubic, tangentMode=RCTM_User, time=0, value=0}
          {interpMode=RCIM_Cubic, tangentMode=RCTM_User, time=0.3, value=1}
          {tangentMode=RCTM_User, time=1, value=0}
         ]
        }}}
      Scale Curve=&RandomRangeFloat(Minimum=0.75, Maximum=1)
     )
     3=?
     Index=&SpawnGroup()
     Num=::NewEnumerator1
    )
   )
   &CurlNoiseForce(
    Bias Noise=true
    Bias Noise Field:Vector3f=&MergeHelper_VectorSwitch(0=<1 0 0.5>, 1=<0 0 0>, 2=<0 0 0>, 3=?, Index=&SpawnGroup(), Num=::NewEnumerator1)
    Noise Frequency=&MergeHelper_FloatSwitch(
     0=12
     1=0
     2=&RandomRangeFloat(Minimum=20, Maximum=30)
     3=?
     Index=&SpawnGroup()
     Num=::NewEnumerator1
    )
    Noise Strength=&MergeHelper_FloatSwitch(
     0=&RandomRangeFloat(Minimum=60, Maximum=175)
     1=0
     2=100
     3=?
     Index=&SpawnGroup()
     Num=::NewEnumerator1
    )
    Pan Noise=false
   )
   &GravityForce(Gravity=&MergeHelper_VectorSwitch(
     0=<0 0 0>
     1=<0 0 0>
     2=&Multiply_VectorByFloat(
      Vector=<0 0 -300>
      Float=&FloatFromCurve(FloatCurve=DataInterface() {propertyValues {Curve {
          defaultValue=1
          keys [
           <at 0.1191 0 RCIM_Cubic>
           {arriveTangent=-0.7473, interpMode=RCIM_Cubic, leaveTangent=-0.7473, tangentMode=RCTM_User, time=0.3761, value=0.9394}
          ]
         }}})
     )
     3=?
     Index=&SpawnGroup()
     Num=::NewEnumerator1
    ))
   &Drag(
    Drag=&MergeHelper_FloatSwitch(
     0=&RandomRangeFloat(Minimum=2, Maximum=5)
     1=&RandomRangeFloat(Minimum=5, Maximum=7)
     2=&RandomRangeFloat(Minimum=0.8, Maximum=2)
     3=?
     Index=&SpawnGroup()
     Num=::NewEnumerator1
    )
    Use Linear Drag=true
    Use Rotational Drag=true
   )
   &SolveForcesAndVelocity(Clamp Velocity=false, Limit Acceleration=false)
   &ScaleColor(
    Scale Alpha=&MergeHelper_FloatSwitch(
     0=&Multiply_Float(
      A=&FloatFromCurve(FloatCurve=DataInterface() {propertyValues {Curve {
          defaultValue=1
          keys [
           {arriveTangent=9.0725, interpMode=RCIM_Cubic, leaveTangent=9.0725, tangentMode=RCTM_User, time=0, value=0}
           {arriveTangent=-2.5249, interpMode=RCIM_Cubic, leaveTangent=-2.5249, tangentMode=RCTM_User, time=0.1408, value=0.9847}
           {arriveTangent=-0.0405, interpMode=RCIM_Cubic, leaveTangent=-0.0405, tangentMode=RCTM_User, time=1, value=0}
          ]
         }}})
      B=&Waveform(
       "Waveform [1]"=::NewEnumerator0
       "[1] Frequency"=&FloatFromCurve(
        FloatCurve=DataInterface() {propertyValues {Curve {keys [<at 0.2 0>,<at 0.4 1>]}}}
        Scale Curve=&RandomRangeFloat(Minimum=3, Maximum=5)
       )
       "[1] Amplitude Min/Max"=<0 1>
      )
     )
     1=&FloatFromCurve(FloatCurve=DataInterface() {propertyValues {Curve {keys [<at 0.6 1 RCIM_Cubic>, {arriveTangent=-3.1779, interpMode=RCIM_Cubic, leaveTangent=-3.1779, tangentMode=RCTM_User, time=1, value=0}]}}})
     2=&Multiply_Float(
      A=&FloatFromCurve(
       FloatCurve=DataInterface() {propertyValues {Curve {keys [<at 0 1 RCIM_Cubic>, {arriveTangent=-2.4387, interpMode=RCIM_Cubic, leaveTangent=-2.4387, tangentMode=RCTM_User, time=1, value=0}]}}}
       Scale Curve=&Waveform("Waveform [1]"=::NewEnumerator0, "[1] Frequency"=4, "[1] Amplitude Min/Max"=<0.1 1.1>)
      )
      B=&FN_FrontEndFloatModifier(Gameplay=1, Front End=0.1)
     )
     3=?
     Index=&SpawnGroup()
     Num=::NewEnumerator1
    )
    Scale RGB=&MergeHelper_VectorSwitch(
     0=<1 1 1>
     1=&Multiply_Vector(
      A=&MakeVectorFromLinearColorRGB(Color=$System.TOD Color)
      B=&VectorFromFloat(Value=&FloatFromCurve(FloatCurve=DataInterface() {propertyValues {Curve {
           defaultValue=1
           keys [
            {arriveTangent=47.9314, interpMode=RCIM_Cubic, leaveTangent=47.9314, tangentMode=RCTM_User, time=0, value=1}
            {arriveTangent=-4.3766, interpMode=RCIM_Cubic, leaveTangent=-4.3766, tangentMode=RCTM_User, time=0.1, value=2.5}
            {
             arriveTangent=-0.6181
             arriveTangentWeight=0.377
             interpMode=RCIM_Cubic
             leaveTangent=-0.6181
             leaveTangentWeight=0.2187
             tangentMode=RCTM_User
             time=1
             value=0.95
            }
           ]
          }}}))
     )
     2=&MakeVectorFromLinearColorRGB(Color=&MultiplyLinearColor(
       Linear Color=&ColorFromCurve(ColorCurve=DataInterface() {propertyValues {
          AlphaCurve {keys []}
          BlueCurve {keys [<at 0 1 RCIM_Cubic>,<at 0.1015 1>,<at 0.118 200>,<at 0.1312 1>,<at 0.306 1>,<at 0.3225 350>,<at 0.3379 1>,<at 1 1 RCIM_Cubic>]}
          GreenCurve {keys [<at 0 1 RCIM_Cubic>,<at 0.1015 1>,<at 0.118 200>,<at 0.1312 1>,<at 0.306 1>,<at 0.3225 350>,<at 0.3379 1>,<at 1 1 RCIM_Cubic>]}
          RedCurve {keys [<at 0 1 RCIM_Cubic>,<at 0.1015 1>,<at 0.118 200>,<at 0.1312 1>,<at 0.306 1>,<at 0.3225 350>,<at 0.3379 1>,<at 1 1 RCIM_Cubic>]}
         }})
       Scale Factor=&MakeVector4_FromLinearColor(LinearColor=&LerpLinearColors(
         StartColor=<rgba 1 1 1 1>
         EndColor=<rgba 20 20 20 1>
         LerpFactor=&Waveform("Waveform [1]"=::NewEnumerator0, "[1] Frequency"=4, "[1] Amplitude Min/Max"=<0 1>)
        ))
      ))
     3=?
     Index=&SpawnGroup()
     Num=::NewEnumerator1
    )
    ScaleA=true
    ScaleRGB=true
   )
   /SetVariables(Particles.StarGlowColor:LinearColor=&MultiplyLinearColor(
     Linear Color=$Particles.Color
     Scale Factor=&MakeVector4FromVectorAndFloat(
      "Vector (RGB)"=<1 1 1>
      "Float (Alpha)"=&FloatFromCurve(FloatCurve=DataInterface() {propertyValues {Curve {
          defaultValue=1
          keys [
           {interpMode=RCIM_Cubic, tangentMode=RCTM_User, time=0.3, value=1}
           {arriveTangent=-0.5449, interpMode=RCIM_Cubic, leaveTangent=-0.5449, tangentMode=RCTM_User, time=1, value=0}
          ]
         }}})
     )
    )) {isSetParameters=true, module=SetVariables_#4}
   &ScaleSpriteSizeBySpeed(
    Max Scale Factor=&MergeHelper_Vector2Switch(0=<1 1>, 1=<1 1>, 2=<1 7>, 3=?, 4=?, Index=&SpawnGroup(), Num=::NewEnumerator1)
    Min Scale Factor=&MergeHelper_Vector2Switch(0=<1 1>, 1=<1 1>, 2=<1 1>, 3=?, 4=?, Index=&SpawnGroup(), Num=::NewEnumerator1)
    Sample Scale Factor By Curve=true
    Scale Factor Curve=DataInterface() {propertyValues {Curve {keys [
        {arriveTangent=2.865, interpMode=RCIM_Cubic, time=0.1247, value=-0.0006}
        {arriveTangent=4.1522, interpMode=RCIM_Cubic, leaveTangent=4.1522, tangentMode=RCTM_User, time=0.3948, value=0.4893}
        <at 0.9958 0.9974 RCIM_Cubic>
       ]}}}
    Velocity Threshold=700
   )
  ]
  rendererClasses [&NiagaraSpriteRendererProperties]
  renderers [
   {rendererClass=&NiagaraSpriteRendererProperties}
   {rendererClass=&NiagaraSpriteRendererProperties, rendererIndex=1}
   {rendererClass=&NiagaraSpriteRendererProperties, rendererIndex=2}
   {rendererClass=&NiagaraSpriteRendererProperties, rendererIndex=3}
  ]
 }
 Glitter_Stars_Flares001 {
  emitterUpdateScript [
   &EmitterState()
   &SpawnBurst_Instantaneous(
    Spawn Count=&RandomRangeInt(Minimum=8, Maximum=10)
    Use Spawn Probability=false
    UseLoopCountLimit=true
   )
   &SpawnBurst_Instantaneous(
    Spawn Count=&RandomRangeInt(Minimum=3, Maximum=5)
    Spawn Group=1
    Use Spawn Probability=false
    UseLoopCountLimit=true
   ) {module=SpawnBurst_Instantaneous001}
   &SpawnBurst_Instantaneous(Spawn Count=17, Spawn Group=2, Use Spawn Probability=false, UseLoopCountLimit=true) {module=SpawnBurst_Instantaneous002}
  ]
  particleSpawnScript [
   /SetVariables(Particles.VisibilityTag=&SpawnGroup()) {isSetParameters=true, module=SetVariables_#2}
   /SetVariables(Particles.Variance:NiagaraFloat=&MergeHelper_FloatSwitch(
     0=&FloatFromCurve(
      FloatCurve=DataInterface() {propertyValues {Curve {keys [<at 0 1>,<at 1 0>]}}}
      CurveIndex=&ReturnNormalizedExecIndex()
     )
     1=&FloatFromCurve(
      FloatCurve=DataInterface() {propertyValues {Curve {keys [<at 0 1>,<at 1 0>]}}}
      CurveIndex=&ReturnNormalizedExecIndex()
     )
     2=0
     3=?
     Index=&SpawnGroup()
     Num=::NewEnumerator1
    )) {isSetParameters=true, module=SetVariables_#3}
   &InitializeParticle(
    Color=&MergeHelper_ColorSwitch(
     0=&MakeCustomLinearColorFromBool(
      Bool=&SetBoolByFloatComparison(A=$User.IsVariant, B=0.01)
      True Color=&ColorFromCurve(ColorCurve=DataInterface() {propertyValues {
         AlphaCurve {keys [<at 0 1 RCIM_Cubic>]}
         BlueCurve {keys [
           <at 0.1342 0.2995>
           {arriveTangent=-0.1943, interpMode=RCIM_Cubic, leaveTangent=-0.1943, time=0.4451, value=0}
           <at 0.7943 0.1712 RCIM_Cubic>
          ]}
         GreenCurve {keys [
           <at 0.1342 0.2995>
           {arriveTangent=-0.4537, interpMode=RCIM_Cubic, leaveTangent=-0.4537, time=0.4451, value=0.0037}
           <at 0.7943 0 RCIM_Cubic>
          ]}
         RedCurve {keys [<at 0.1342 5>, {arriveTangent=-5.3023, interpMode=RCIM_Cubic, leaveTangent=-5.3023, time=0.4451, value=3}, <at 0.7943 1.5 RCIM_Cubic>]}
        }})
      False Color=&ColorFromCurve(ColorCurve=DataInterface() {propertyValues {
         AlphaCurve {keys [<at 0 1 RCIM_Cubic>]}
         BlueCurve {keys [<at 0.1342 3.3207>,<at 0.521 1.9924 RCIM_Cubic>,<at 0.7352 1.1029 RCIM_Cubic>]}
         GreenCurve {keys [<at 0.1342 0.5581>,<at 0.521 0.3349 RCIM_Cubic>,<at 0.7352 0 RCIM_Cubic>]}
         RedCurve {keys [<at 0.1342 5>,<at 0.521 3 RCIM_Cubic>,<at 0.7352 1.5 RCIM_Cubic>]}
        }})
     )
     1=&MakeCustomLinearColorFromBool(
      Bool=&SetBoolByFloatComparison(A=$User.IsVariant, B=0.01)
      True Color=<rgba 1.15 0.1096 0.092 1>
      False Color=<rgba 1 0.1274 0.402 1>
     )
     2=&MakeCustomLinearColorFromBool(
      Bool=&SetBoolByFloatComparison(A=$User.IsVariant, B=0)
      True Color=&SelectLinearColorFromArray(
       Array Sampling Mode=::NewEnumerator0
       Direct Array Index=?
       Color Selection Array=DataInterface() {propertyValues {ColorData [
          <rgba 1 0.1199 0.0714 1>
          <rgba 1 0.0571 0.0571 1>
          <rgba 0.1473 0.1405 0.149 1>
          <rgba 1 0.112 0.112 1>
          <rgba 0.2406 0.0526 0.0835 1>
          <rgba 1 0.1043 0.1043 1>
          <rgba 0.7698 0.0003 0.0404 1>
          <rgba 0.4281 0 0 1>
         ]}}
       Random Seed=0
      )
      False Color=&SelectLinearColorFromArray(
       Array Sampling Mode=::NewEnumerator0
       Direct Array Index=?
       Color Selection Array=DataInterface() {propertyValues {ColorData [
          <rgba 1 0.0714 0.5244 1>
          <rgba 0.5687 0.0571 1 1>
          <rgba 0.1473 0.1405 0.149 1>
          <rgba 0.8187 0.1885 1 1>
          <rgba 0.1226 0.0724 0.2406 1>
          <rgba 0.4133 0.1594 1 1>
          <rgba 0.7698 0.0003 0.6263 1>
          <rgba 0.4281 0 0.062 1>
         ]}}
       Random Seed=0
      )
     )
     3=?
     Index=&SpawnGroup()
     Num=::NewEnumerator1
    )
    Lifetime=&MergeHelper_FloatSwitch(
     0=&RandomRangeFloat(Minimum=0.45, Maximum=1)
     1=&FloatFromCurve(FloatCurve=DataInterface() {propertyValues {Curve {keys [<at 0 0.58>,<at 1 0.33>]}}}, CurveIndex=$Particles.Variance)
     2=&RandomRangeFloat(Minimum=0.5, Maximum=0.8)
     3=?
     Index=&SpawnGroup()
     Num=::NewEnumerator1
    )
    Sprite Rotation Angle=&MergeHelper_FloatSwitch(
     0=45
     1=&RandomRangeFloat(Minimum=0, Maximum=360)
     2=0
     3=?
     Index=&SpawnGroup()
     Num=::NewEnumerator1
    )
    Sprite Rotation Mode=::NewEnumerator0
    Sprite Size Mode=::NewEnumerator3
    Uniform Sprite Size=&MergeHelper_FloatSwitch(
     0=&RandomRangeFloat(Minimum=5, Maximum=9)
     1=&FloatFromCurve(
      FloatCurve=DataInterface() {propertyValues {Curve {keys [<at 0 5>,<at 1 9>]}}}
      CurveIndex=$Particles.Variance
      Scale Curve=&FN_FrontEndFloatModifier(Gameplay=1.9, Front End=1.5)
     )
     2=&Multiply_Float(A=&RandomRangeFloat(Minimum=1.5, Maximum=5.5), B=$System.GlobalScale)
     3=?
     Index=&SpawnGroup()
     Num=::NewEnumerator1
    )
    UsePositionOffset=false
    Write Material Random=true
   )
   &ShapeLocation(
    MaskSpawnGroup=false
    Rotation Angle:NiagaraFloat=&RandomRangeFloat(Minimum=0, Maximum=1)
    Rotation Axis=<0 1 1>
    Rotation Mode=::NewEnumerator1
    Sphere Distribution=::NewEnumerator3
    Sphere Radius=&Multiply_Float(
     A=&MergeHelper_FloatSwitch(0=15, 1=20, 2=7, 3=?, Index=&SpawnGroup(), Num=::NewEnumerator1)
     B=$System.GlobalScale
    )
    UseOwnerScale=true
   )
   &AddVelocity(
    Module.Use Falloff Distance=false
    Override Seed=false
    Velocity Mode=::NewEnumerator1
    Velocity Speed=&Multiply_Float(
     A=&MergeHelper_FloatSwitch(
      0=&RandomRangeFloat(Minimum=75, Maximum=200)
      1=&FloatFromCurve(FloatCurve=DataInterface() {propertyValues {Curve {keys [<at 0 175>,<at 1 100>]}}}, CurveIndex=$Particles.Variance)
      2=&RandomRangeFloat(Minimum=75, Maximum=150)
      3=?
      Index=&SpawnGroup()
      Num=::NewEnumerator1
     )
     B=$System.GlobalScale
    )
   )
   &CameraOffset(Camera Offset Amount=15)
  ]
  particleUpdateScript [
   &ParticleState()
   &ScaleSpriteSize(
    Scale Sprite Size Mode=::NewEnumerator0
    Uniform Scale Factor=&MergeHelper_FloatSwitch(
     0=&FloatFromCurve(
      FloatCurve=DataInterface() {propertyValues {Curve {
         defaultValue=1
         keys [
          {arriveTangent=9.0725, interpMode=RCIM_Cubic, leaveTangent=9.0725, tangentMode=RCTM_User, time=0, value=0}
          {arriveTangent=-2.5249, interpMode=RCIM_Cubic, leaveTangent=-2.5249, tangentMode=RCTM_User, time=0.1408, value=0.9847}
          {arriveTangent=-0.0405, interpMode=RCIM_Cubic, leaveTangent=-0.0405, tangentMode=RCTM_User, time=1, value=0}
         ]
        }}}
      Scale Curve=$System.GlobalScale
     )
     1=&FloatFromCurve(
      FloatCurve=DataInterface() {propertyValues {Curve {
         defaultValue=1
         keys [
          {arriveTangent=3.6998, interpMode=RCIM_Cubic, leaveTangent=3.6998, tangentMode=RCTM_User, time=0, value=0.2}
          {arriveTangent=0.2776, interpMode=RCIM_Cubic, leaveTangent=0.2776, tangentMode=RCTM_User, time=0.55, value=1}
          {arriveTangent=-4.5068, interpMode=RCIM_Cubic, leaveTangent=-4.5068, tangentMode=RCTM_User, time=0.6, value=1.2}
          {arriveTangent=-2.5817, interpMode=RCIM_Cubic, leaveTangent=-2.5817, tangentMode=RCTM_User, time=1, value=0}
         ]
        }}}
      Scale Curve=$System.GlobalScale
     )
     2=&FloatFromCurve(
      FloatCurve=DataInterface() {propertyValues {Curve {
         defaultValue=0
         keys [
          {interpMode=RCIM_Cubic, tangentMode=RCTM_User, time=0, value=0}
          {interpMode=RCIM_Cubic, tangentMode=RCTM_User, time=0.3, value=1}
          {tangentMode=RCTM_User, time=1, value=0}
         ]
        }}}
      Scale Curve=&RandomRangeFloat(Minimum=0.75, Maximum=1)
     )
     3=?
     Index=&SpawnGroup()
     Num=::NewEnumerator1
    )
   )
   &CurlNoiseForce(
    Bias Noise=true
    Bias Noise Field:Vector3f=&MergeHelper_VectorSwitch(0=<1 0 0.5>, 1=<0 0 0>, 2=<0 0 0>, 3=?, Index=&SpawnGroup(), Num=::NewEnumerator1)
    Noise Frequency=&MergeHelper_FloatSwitch(
     0=12
     1=0
     2=&RandomRangeFloat(Minimum=20, Maximum=30)
     3=?
     Index=&SpawnGroup()
     Num=::NewEnumerator1
    )
    Noise Strength=&MergeHelper_FloatSwitch(
     0=&RandomRangeFloat(Minimum=60, Maximum=175)
     1=0
     2=100
     3=?
     Index=&SpawnGroup()
     Num=::NewEnumerator1
    )
    Pan Noise=false
   )
   &GravityForce(Gravity=&MergeHelper_VectorSwitch(
     0=<0 0 0>
     1=<0 0 0>
     2=&Multiply_VectorByFloat(
      Vector=<0 0 -300>
      Float=&FloatFromCurve(FloatCurve=DataInterface() {propertyValues {Curve {
          defaultValue=1
          keys [
           <at 0.1191 0 RCIM_Cubic>
           {arriveTangent=-0.7473, interpMode=RCIM_Cubic, leaveTangent=-0.7473, tangentMode=RCTM_User, time=0.3761, value=0.9394}
          ]
         }}})
     )
     3=?
     Index=&SpawnGroup()
     Num=::NewEnumerator1
    ))
   &Drag(
    Drag=&MergeHelper_FloatSwitch(
     0=&RandomRangeFloat(Minimum=2, Maximum=5)
     1=&RandomRangeFloat(Minimum=5, Maximum=7)
     2=&RandomRangeFloat(Minimum=0.8, Maximum=2)
     3=?
     Index=&SpawnGroup()
     Num=::NewEnumerator1
    )
    Use Linear Drag=true
    Use Rotational Drag=true
   )
   &SolveForcesAndVelocity(Clamp Velocity=false, Limit Acceleration=false)
   &ScaleColor(
    Scale Alpha=&MergeHelper_FloatSwitch(
     0=&Multiply_Float(
      A=&FloatFromCurve(FloatCurve=DataInterface() {propertyValues {Curve {
          defaultValue=1
          keys [
           {arriveTangent=9.0725, interpMode=RCIM_Cubic, leaveTangent=9.0725, tangentMode=RCTM_User, time=0, value=0}
           {arriveTangent=-2.5249, interpMode=RCIM_Cubic, leaveTangent=-2.5249, tangentMode=RCTM_User, time=0.1408, value=0.9847}
           {arriveTangent=-0.0405, interpMode=RCIM_Cubic, leaveTangent=-0.0405, tangentMode=RCTM_User, time=1, value=0}
          ]
         }}})
      B=&Waveform(
       "Waveform [1]"=::NewEnumerator0
       "[1] Frequency"=&FloatFromCurve(
        FloatCurve=DataInterface() {propertyValues {Curve {keys [<at 0.2 0>,<at 0.4 1>]}}}
        Scale Curve=&RandomRangeFloat(Minimum=3, Maximum=5)
       )
       "[1] Amplitude Min/Max"=<0 1>
      )
     )
     1=&FloatFromCurve(FloatCurve=DataInterface() {propertyValues {Curve {keys [<at 0.6 1 RCIM_Cubic>, {arriveTangentWeight=0.4442, interpMode=RCIM_Cubic, tangentMode=RCTM_User, time=1, value=0}]}}})
     2=&Multiply_Float(
      A=&FloatFromCurve(
       FloatCurve=DataInterface() {propertyValues {Curve {keys [<at 0 1 RCIM_Cubic>, {arriveTangent=-2.4387, interpMode=RCIM_Cubic, leaveTangent=-2.4387, tangentMode=RCTM_User, time=1, value=0}]}}}
       Scale Curve=&Waveform("Waveform [1]"=::NewEnumerator0, "[1] Frequency"=4, "[1] Amplitude Min/Max"=<0.1 1.1>)
      )
      B=&FN_FrontEndFloatModifier(Gameplay=1, Front End=0.1)
     )
     3=?
     Index=&SpawnGroup()
     Num=::NewEnumerator1
    )
    Scale RGB=&MergeHelper_VectorSwitch(
     0=<1 1 1>
     1=&Multiply_Vector(
      A=&MakeVectorFromLinearColorRGB(Color=$System.TOD Color)
      B=&VectorFromFloat(Value=&FloatFromCurve(FloatCurve=DataInterface() {propertyValues {Curve {
           defaultValue=1
           keys [
            {arriveTangent=47.9314, interpMode=RCIM_Cubic, leaveTangent=47.9314, tangentMode=RCTM_User, time=0, value=1}
            {arriveTangent=-4.3766, interpMode=RCIM_Cubic, leaveTangent=-4.3766, tangentMode=RCTM_User, time=0.1, value=2.5}
            {
             arriveTangent=-0.6181
             arriveTangentWeight=0.377
             interpMode=RCIM_Cubic
             leaveTangent=-0.6181
             leaveTangentWeight=0.2187
             tangentMode=RCTM_User
             time=1
             value=0.95
            }
           ]
          }}}))
     )
     2=&MakeVectorFromLinearColorRGB(Color=&MultiplyLinearColor(
       Linear Color=&ColorFromCurve(ColorCurve=DataInterface() {propertyValues {
          AlphaCurve {keys []}
          BlueCurve {keys [<at 0 1 RCIM_Cubic>,<at 0.1015 1>,<at 0.118 200>,<at 0.1312 1>,<at 0.306 1>,<at 0.3225 350>,<at 0.3379 1>,<at 1 1 RCIM_Cubic>]}
          GreenCurve {keys [<at 0 1 RCIM_Cubic>,<at 0.1015 1>,<at 0.118 200>,<at 0.1312 1>,<at 0.306 1>,<at 0.3225 350>,<at 0.3379 1>,<at 1 1 RCIM_Cubic>]}
          RedCurve {keys [<at 0 1 RCIM_Cubic>,<at 0.1015 1>,<at 0.118 200>,<at 0.1312 1>,<at 0.306 1>,<at 0.3225 350>,<at 0.3379 1>,<at 1 1 RCIM_Cubic>]}
         }})
       Scale Factor=&MakeVector4_FromLinearColor(LinearColor=&LerpLinearColors(
         StartColor=<rgba 1 1 1 1>
         EndColor=<rgba 20 20 20 1>
         LerpFactor=&Waveform("Waveform [1]"=::NewEnumerator0, "[1] Frequency"=4, "[1] Amplitude Min/Max"=<0 1>)
        ))
      ))
     3=?
     Index=&SpawnGroup()
     Num=::NewEnumerator1
    )
    ScaleA=true
    ScaleRGB=true
   )
   /SetVariables(Particles.StarGlowColor:LinearColor=&MultiplyLinearColor(
     Linear Color=$Particles.Color
     Scale Factor=&MakeVector4FromVectorAndFloat(
      "Vector (RGB)"=<1 1 1>
      "Float (Alpha)"=&FloatFromCurve(FloatCurve=DataInterface() {propertyValues {Curve {
          defaultValue=1
          keys [
           {interpMode=RCIM_Cubic, tangentMode=RCTM_User, time=0.3, value=1}
           {arriveTangent=-0.5449, interpMode=RCIM_Cubic, leaveTangent=-0.5449, tangentMode=RCTM_User, time=1, value=0}
          ]
         }}})
     )
    )) {isSetParameters=true, module=SetVariables_#4}
   &ScaleSpriteSizeBySpeed(
    Max Scale Factor=&MergeHelper_Vector2Switch(0=<1 1>, 1=<1 1>, 2=<1 7>, 3=?, 4=?, Index=&SpawnGroup(), Num=::NewEnumerator1)
    Min Scale Factor=&MergeHelper_Vector2Switch(0=<1 1>, 1=<1 1>, 2=<1 1>, 3=?, 4=?, Index=&SpawnGroup(), Num=::NewEnumerator1)
    Sample Scale Factor By Curve=true
    Scale Factor Curve=DataInterface() {propertyValues {Curve {keys [
        {arriveTangent=2.865, interpMode=RCIM_Cubic, time=0.1247, value=-0.0006}
        {arriveTangent=4.1522, interpMode=RCIM_Cubic, leaveTangent=4.1522, tangentMode=RCTM_User, time=0.3948, value=0.4893}
        <at 0.9958 0.9974 RCIM_Cubic>
       ]}}}
    Velocity Threshold=700
   )
  ]
  rendererClasses [&NiagaraSpriteRendererProperties]
  renderers [
   {rendererClass=&NiagaraSpriteRendererProperties}
   {rendererClass=&NiagaraSpriteRendererProperties, rendererIndex=1}
   {rendererClass=&NiagaraSpriteRendererProperties, rendererIndex=2}
   {rendererClass=&NiagaraSpriteRendererProperties, rendererIndex=3}
  ]
 }
 Impact_Smoke {
  emitterUpdateScript [
   &EmitterState()
   &SpawnBurst_Instantaneous(
    Spawn Count=&RandomRangeInt(Minimum=2, Maximum=3)
    Use Spawn Probability=false
    UseLoopCountLimit=false
   )
  ]
  particleSpawnScript [
   &InitializeParticle(
    Color Maximum=&MakeCustomLinearColorFromBool(
     Bool=&SetBoolByFloatComparison(A=$User.IsVariant, B=0.01)
     True Color=<rgba 1 0.3333 0.3333 0.25>
     False Color=<rgba 0.8555 0.402 1 0.25>
    )
    Color Minimum=&MakeCustomLinearColorFromBool(
     Bool=&SetBoolByFloatComparison(A=$User.IsVariant, B=0.01)
     True Color=<rgba 1 0.3988 0.3797 0.2>
     False Color=<rgba 1 0.4271 0.7073 0.2>
    )
    Color Mode=::NewEnumerator2
    Lifetime Max=0.45
    Lifetime Min=0.35
    Lifetime Mode=::NewEnumerator1
    Sprite Size Mode=::NewEnumerator3
    Uniform Sprite Size=&Multiply_Float(A=$System.GlobalScale, B=&RandomRangeFloat(Minimum=100, Maximum=130))
    UsePositionOffset=false
    Write Material Random=true
   )
   &CameraOffset(Camera Offset Amount=&RandomRangeFloat(Minimum=40, Maximum=50, Evaluation Type=::NewEnumerator1))
   &ShapeLocation(
    Hemisphere Distribution=<1 0.5>
    MaskSpawnGroup=false
    Override Seed=false
    Rotation Angle=90
    Rotation Mode=::NewEnumerator1
    Sphere Radius=&MakeCustomFloatFromBool(Bool=$User.bIsFrontEndPreview, True Float=5, False Float=10)
    UseOwnerScale=true
   )
   &AddVelocity(Velocity=&Multiply_VectorByFloat(Vector=$Particles.ShapeLocation.ShapeVector, Float=10))
  ]
  particleUpdateScript [
   &ParticleState()
   &ScaleSpriteSize(Uniform Curve Sprite Scale=DataInterface() {propertyValues {Curve {keys [
        {arriveTangent=7.8101, interpMode=RCIM_Cubic, leaveTangent=7.8101, tangentMode=RCTM_User, time=0, value=0}
        {arriveTangent=1, interpMode=RCIM_Cubic, leaveTangent=1, time=0.1505, value=0.6088}
        {interpMode=RCIM_Cubic, tangentMode=RCTM_User, time=1, value=1}
       ]}}})
   &ScaleColor(
    Scale Alpha=&FloatFromCurve(
     FloatCurve=DataInterface() {propertyValues {Curve {keys [
         {arriveTangent=4.3774, interpMode=RCIM_Cubic, leaveTangent=4.3774, tangentMode=RCTM_User, time=0, value=0}
         <at 0.2 0.5 RCIM_Cubic>
         <at 1 0 RCIM_Cubic>
        ]}}}
     Scale Curve=&FN_FrontEndFloatModifier(Gameplay=1, Front End=0.6)
    )
    Scale RGB=&MakeVectorFromLinearColorRGB(Color=$System.TOD Color)
    ScaleA=true
    ScaleRGB=true
   )
   &DynamicMaterialParameters(
    Index 0 Param 1=&Add_Float(
     A=&RandomRangeFloat(Minimum=0, Maximum=1)
     B=&FloatFromCurve(
      FloatCurve=DataInterface() {propertyValues {Curve {keys [
          <at 0 0 RCIM_Cubic>
          {arriveTangent=1.5947, interpMode=RCIM_Cubic, leaveTangent=1.5947, tangentMode=RCTM_User, time=0.2557, value=0.5257}
          <at 1 1 RCIM_Cubic>
         ]}}}
      CurveIndex=$Particles.Age
      Scale Curve=0.25
     )
    )
    Module.Index0Float0Write=true
    Module.Index0Float1Write=true
    Module.Index0Float2Write=true
    Module.Index0Float3Write=true
   ) {module=DynamicMaterialParameters001}
   &SolveForcesAndVelocity(Clamp Velocity=false, Limit Acceleration=false)
  ]
  rendererClasses [&NiagaraSpriteRendererProperties]
  renderers [
   {rendererClass=&NiagaraSpriteRendererProperties}
   {rendererClass=&NiagaraSpriteRendererProperties, rendererIndex=1}
  ]
 }
 Toon_Spikes {
  emitterUpdateScript [
   &EmitterState()
   &SpawnBurst_Instantaneous(Spawn Count=13, Use Spawn Probability=false)
  ]
  particleSpawnScript [
   &InitializeParticle(
    Color=&MakeCustomLinearColorFromBool(
     Bool=&SetBoolByFloatComparison(A=$User.IsVariant, B=0.01)
     True Color=&SelectLinearColorFromArray(
      Array Sampling Mode=::NewEnumerator0
      Direct Array Index=?
      Color Selection Array=DataInterface() {propertyValues {ColorData [<rgba 0.0443 0.0158 0.0181 1>, <rgba 0.0443 0.0243 0.0241 1>, <rgba 0.0443 0.0236 0.0191 1>]}}
      Random Seed=0
     )
     False Color=&SelectLinearColorFromArray(
      Array Sampling Mode=::NewEnumerator0
      Direct Array Index=?
      Color Selection Array=DataInterface() {propertyValues {ColorData [<rgba 0.0309 0.0297 0.0443 1>, <rgba 0.0443 0.0329 0.0382 1>, <rgba 0.0248 0.0286 0.0443 1>]}}
      Random Seed=0
     )
    )
    Lifetime Max=0.18
    Lifetime Min=0.12
    Lifetime Mode=::NewEnumerator1
    Sprite Rotation Mode=::NewEnumerator2
    Sprite Size Max=<9 45>
    Sprite Size Min=<5 35>
    Sprite Size Mode=::NewEnumerator2
    UsePositionOffset=false
    Write Material Random=true
   ) {module=InitializeParticle001}
   &ShapeLocation(
    MaskSpawnGroup=false
    Rotation Angle:NiagaraFloat=&RandomRangeFloat(Minimum=0, Maximum=1)
    Rotation Axis=<0 0 1>
    Rotation Mode=::NewEnumerator1
    Sphere Distribution=::NewEnumerator3
    Sphere Radius=&Multiply_Float(A=2, B=$System.GlobalScale)
    UseOwnerScale=true
   )
   &AddVelocityFromPoint(Module.Use Falloff Distance=false, Velocity Strength=&RandomRangeFloat(Minimum=100, Maximum=200))
  ]
  particleUpdateScript [
   &CameraOffset(Camera Offset Amount=15)
   &ParticleState()
   &SolveForcesAndVelocity(Clamp Velocity=false, Limit Acceleration=false)
   &ScaleSpriteSize(Scale Factor=&Vector2DFromCurve(
     Vector2Curve=DataInterface() {propertyValues {XCurve {keys [<at 0 0.778>,<at 1 1>]}, YCurve {keys [<at 0 0.03>,<at 1 1.5>]}}}
     ScaleCurve=&Vector2DFromFloat(Value=$System.GlobalScale)
    ))
   &ScaleColor(
    Linear Color Curve=DataInterface() {propertyValues {
      AlphaCurve {keys [<at 0.6553 1>,<at 1 0>]}
      BlueCurve {keys [<at 0 2>,<at 0.1623 1>]}
      GreenCurve {keys [<at 0 2>,<at 0.1623 1>]}
      RedCurve {keys [<at 0 2>,<at 0.1623 1>]}
     }}
    Scale Mode=::NewEnumerator2
   )
   &ScaleColor(
    Scale Alpha=&FN_FrontEndFloatModifier(Gameplay=1, Front End=1)
    Scale RGB=&MakeVectorFromLinearColorRGB(Color=$System.TOD Color)
    ScaleA=true
    ScaleRGB=true
   ) {module=ScaleColor001}
  ]
  rendererClasses [&NiagaraSpriteRendererProperties]
  renderers [{rendererClass=&NiagaraSpriteRendererProperties}]
 }
}
summary {
 systemName=NS_Pickaxe_MicStand_ForestJeanWink_Impact
 emitters [
  {
   emitterName=FN_PhysMatDebris
   rendererClasses [&NiagaraSpriteRendererProperties, &NiagaraMeshRendererProperties]
  }
  {emitterName=Impact_Smoke, rendererClasses [&NiagaraSpriteRendererProperties]}
  {emitterName=Toon_Spikes, rendererClasses [&NiagaraSpriteRendererProperties]}
  {
   emitterName=Flash
   rendererClasses [
    &NiagaraMeshRendererProperties
    &NiagaraSpriteRendererProperties
    &NiagaraLightRendererProperties
   ]
  }
  {bEnabled=false, emitterName=Glitter_Stars_Flares, rendererClasses [&NiagaraSpriteRendererProperties]}
  {emitterName=Glitter_Stars_Flares001, rendererClasses [&NiagaraSpriteRendererProperties]}
 ]
 userVariables [
  User.bIsFrontEnd {type=NiagaraBool}
  User.bIsFrontEndPreview {type=NiagaraBool}
  User.IsVariant {type=NiagaraFloat, defaultValue=0, description=""}
  User.PhysicalSurface {type=NiagaraInt32}
 ]
}
