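# Niagara ribbon trail for the ForestJeanWink MicStand pickaxe cosmetic, built on the AnimTrail modules sampling a skeletal mesh.
# One emitter, Chromatic_Trail, spawns short-lived particles whose color switches on User.IsVariant, drawn by two ribbon renderers.
# Curves taper width and fade alpha, with System.TOD Color tinting RGB and a front-end modifier dimming alpha.
# A separate Particles.Linear Color gets its own fade with the same tint and dimming.
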NiagaraSystem /SparksCosmetics/Instrument/MicStand/ForestJeanWink/Effects/NS_Pickaxe_MicStand_ForestJeanWink_Trail
dependencies {
 usedDataInterfaces [&NiagaraDataInterfaceCurve, &NiagaraDataInterfaceColorCurve]
 usedDynamicInputs [&ColorFromCurve]
 usedModules [
  &self:SystemSpawnScript.SystemScriptSource.SystemScriptGraph.NiagaraNodeAssignment_0.SetVariables_#1
  &SystemState
 ]
}
emitters {Chromatic_Trail {
  emitterUpdateScript [
   &EmitterState(Life Cycle Mode=::NewEnumerator1, Loop Behavior=::NewEnumerator1, Loop Duration=0.25, UseLoopDelay=true)
   &FN_SpawnAnimTrail_v2()
  ]
  particleSpawnScript [
   /Game/Effects/Niagara/Modules/AnimTrail/FN_InitializeAnimTrail/FN_InitializeAnimTrailBoneIndecies(First Bone or Socket Index=1)
   /Game/Effects/Niagara/Modules/AnimTrail/FN_InitializeAnimTrail/FN_InitializeAnimTrail(
    AnimTrailWidth=1
    Color=&MakeCustomLinearColorFromBool(
     Bool=&SetBoolByFloatComparison(A=$User.IsVariant, B=0.01)
     True Color=<rgba 0.4948 0 0.0062 1.2>
     False Color=<rgba 0.7552 0.1554 0.5798 1>
    )
    Lifetime=0.22
    Position Offset:Vector3f=<0 0 -5>
   )
   &SystemLocation(MaskSpawnGroup=false)
   &DynamicMaterialParameters(
    Index 0 Param 1=1
    Index 0 Param 4=1
    Module.Index0Param0Write=true
    Module.Index0Param1Write=true
    Module.Index0Param2Write=true
    Module.Index0Param3Write=true
   )
   &AddVelocity(Velocity=<500 0 0>) {enabled=false}
  ]
  particleUpdateScript [
   &ScaleRibbonWidth(Ribbon Width Scale=&FloatFromCurve(FloatCurve=DataInterface() {propertyValues {Curve {keys [
         {interpMode=RCIM_Cubic, leaveTangentWeight=0.7648, tangentMode=RCTM_User, time=0.3, value=1}
         {arriveTangentWeight=0.3245, interpMode=RCIM_Cubic, tangentMode=RCTM_User, time=1, value=0}
        ]}}}))
   &ScaleColor(
    Scale Alpha=&Multiply_Float(
     A=&FloatFromCurve(
      FloatCurve=DataInterface() {propertyValues {Curve {keys [
          {interpMode=RCIM_Cubic, leaveTangent=-3.8577, tangentMode=RCTM_Break, time=0, value=1}
          {arriveTangent=-0.8495, interpMode=RCIM_Cubic, leaveTangent=-0.8495, tangentMode=RCTM_User, time=0.2913, value=0.2}
          {arriveTangent=-0.3792, arriveTangentWeight=0.2108, interpMode=RCIM_Cubic, leaveTangent=-0.3792, tangentMode=RCTM_User, time=1, value=0}
         ]}}}
      Scale Curve=&FloatFromCurve(
       FloatCurve=DataInterface() {propertyValues {Curve {keys [<at 0 1>,<at 0.25 1>,<at 0.5 1>,<at 0.75 1>,<at 1 1>]}}}
       CurveIndex=Linked() {linkedVariable {name=NPC.FortniteNPC.TimeOfDay_0-1, type=NiagaraFloat}}
      )
     )
     B=&FN_FrontEndFloatModifier(Gameplay=1, Front End=0.6)
    )
    Scale RGB=&MakeVectorFromLinearColorRGB(Color=$System.TOD Color)
    ScaleA=true
    ScaleRGB=true
   ) {module=ScaleColor002}
   &FN_UpdateAnimTrail_v2()
   /SetVariables(Particles.Linear Color=&MultiplyLinearColor(
     Linear Color=<rgba 0.5753 0.5311 0.7005 1>
     Scale Factor=&MakeVector4FromVectorAndFloat(
      "Vector (RGB)"=&MakeVectorFromLinearColorRGB(Color=$System.TOD Color)
      "Float (Alpha)"=&Multiply_Float(
       A=&FloatFromCurve(
        FloatCurve=DataInterface() {propertyValues {Curve {keys [
            {interpMode=RCIM_Cubic, leaveTangent=-2.3675, tangentMode=RCTM_Break, time=0, value=1}
            {arriveTangent=-1.3779, interpMode=RCIM_Cubic, leaveTangent=-1.3779, tangentMode=RCTM_User, time=0.3258, value=0.4341}
            {arriveTangent=-0.2399, arriveTangentWeight=0.5252, interpMode=RCIM_Cubic, leaveTangent=-0.2399, tangentMode=RCTM_User, time=1, value=0}
           ]}}}
        Scale Curve=&FloatFromCurve(
         FloatCurve=DataInterface() {propertyValues {Curve {keys [<at 0 1>,<at 0.25 1>,<at 0.5 1>,<at 0.75 1>,<at 1 1>]}}}
         CurveIndex=Linked() {linkedVariable {name=NPC.FortniteNPC.TimeOfDay_0-1, type=NiagaraFloat}}
        )
       )
       B=&FN_FrontEndFloatModifier(Gameplay=1, Front End=0.6)
      )
     )
    )) {isSetParameters=true}
   &SolveForcesAndVelocity(Clamp Velocity=false, Limit Acceleration=false) {enabled=false}
  ]
  rendererClasses [&NiagaraRibbonRendererProperties]
  renderers [
   {rendererClass=&NiagaraRibbonRendererProperties}
   {rendererClass=&NiagaraRibbonRendererProperties, rendererIndex=1}
  ]
 }}
summary {
 systemName=NS_Pickaxe_MicStand_ForestJeanWink_Trail
 emitters [{emitterName=Chromatic_Trail, rendererClasses [&NiagaraRibbonRendererProperties]}]
 userVariables [
  User.bIsFrontEnd {type=NiagaraBool}
  User.bIsFrontEndPreview {type=NiagaraBool}
  User.IsVariant {type=NiagaraFloat, defaultValue=0, description=""}
  User.RibbonIDOffset {type=NiagaraInt32}
  User.SkeletalMesh {
   type=NiagaraDataInterfaceSkeletalMesh
   defaultValue {
    struct=NiagaraExt_VariableValue_DataInterface
    value {dataInterface=&self:NiagaraDataInterfaceSkeletalMesh_0}
   }
  }
  User.TrailSecondaryColor {type=LinearColor, description="", defaultValue {struct=LinearColor, value=<rgba 10 10 10 0.25>}}
 ]
}
