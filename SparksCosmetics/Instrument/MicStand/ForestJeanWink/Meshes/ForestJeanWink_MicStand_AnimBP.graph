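# Anim Blueprint for the ForestJeanWink mic stand instrument cosmetic, adding secondary physics motion to a dangling string and ball.
# Three chained Anim Dynamics nodes simulate dyn_string_01, dyn_string_02 and dyn_ball with angular limits, scaled component motion and a gravity override.
# The Event Graph sets a GravityOverride variable from skydiving and parachute state, flipping it upward in freefall.
# The Anim Dynamics nodes show fixed downward gravity, with no visible link to that variable.
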AnimBlueprint /SparksCosmetics/Instrument/MicStand/ForestJeanWink/Meshes/ForestJeanWink_MicStand_AnimBP
target_skeleton=&/ForestJeanWink_MicStand_Skeleton
graphs [
 AnimGraph {nodes [
   n1 = Anim Dynamics - Bone: dyn_string_01!(ComponentPose=Local To Component(LocalPose=Input Pose()), Alpha=1, GravityOverride:Vector="0,0,-980", ExternalForce="0,0,0") -> (Pose) {
    angularDampingOverride=0.99
    bOverrideAngularDamping=true
    bUseGravityOverride=true
    alphaBoolBlend {blendOption=Linear}
    boundBone {boneName=dyn_string_01}
    componentAppliedLinearAccClamp=<500 500 500>
    componentLinearAccScale=<0.5 0.5 0.5>
    componentLinearVelScale=<0.5 0.5 0.8>
    physicsBodyDefinitions [{
      boundBone {boneName=dyn_string_01}
      boxExtents=<8 8 8>
      constraintSetup {angularLimitsMax=<90 90 90>, angularLimitsMin=<-80 -80 -80>, angularTarget=<0 0 0>}
      localJointOffset=<0.5 0 0>
     }]
    relativeSpaceBone {boneName=handle}
   }
   n2 = Anim Dynamics - Bone: dyn_string_02!(ComponentPose=@n1, Alpha=1, GravityOverride:Vector="0,0,-980", ExternalForce="0,0,0") -> (Pose) {
    angularDampingOverride=0.99
    bOverrideAngularDamping=true
    bUseGravityOverride=true
    linearDampingOverride=0.9
    alphaBoolBlend {blendOption=Linear}
    boundBone {boneName=dyn_string_02}
    componentAppliedLinearAccClamp=<1000 1000 1000>
    componentLinearAccScale=<0.6 0.6 0.85>
    componentLinearVelScale=<0.6 0.6 0.8>
    physicsBodyDefinitions [{
      boundBone {boneName=dyn_string_02}
      boxExtents=<10 1 10>
      constraintSetup {angularLimitsMax=<15 55 45>, angularLimitsMin=<-15 -60 -45>, angularTarget=<0 0 0>}
      localJointOffset=<0.5 0 0>
     }]
   }
   n3 = Anim Dynamics - Bone: dyn_ball!(ComponentPose=@n2, Alpha=1, GravityOverride:Vector="0,0,-980", ExternalForce="0,0,0") -> (Pose) {
    angularDampingOverride=0.99
    bOverrideAngularDamping=true
    bUseGravityOverride=true
    linearDampingOverride=0.9
    alphaBoolBlend {blendOption=Linear}
    boundBone {boneName=dyn_ball}
    componentAppliedLinearAccClamp=<1100 1100 1100>
    componentLinearAccScale=<0.6 0.6 0.85>
    componentLinearVelScale=<0.6 0.6 0.8>
    physicsBodyDefinitions [{
      boundBone {boneName=dyn_ball}
      boxExtents=<8 8 8>
      constraintSetup {angularLimitsMax=<15 30 45>, angularLimitsMin=<-15 -25 -45>, angularTarget=<0 0 0>, linearAxesMin=<-0.25 -0.25 -0.25>}
      localJointOffset=<1 0 0>
     }]
   }
   Output Pose(Result=Component To Local(ComponentPose=@n3))
  ]}
 EventGraph {nodes [
   Event BlueprintUpdateAnimation()
   > n1 = Branch(Condition=Get bIsSkyDiving())
   > n2 = Branch(Condition=Get bIsParachuteOpened())
   > Set GravityOverride(GravityOverride="0,0,-100")
   Set GravityOverride(execute=@n2.else, GravityOverride="0,0,100")
   Set GravityOverride(execute=@n1.else, GravityOverride="0,0,-980")
   TryGetPawnOwner()
  ]}
]
variables [GravityOverride {type=Vector}]
